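# Material function for the Banana Adventure backpack that makes its look react to time of day, changing the incoming material attributes at night.
# Red and green channels of the Tech Art Mask get separate strength and color parameters, are combined by Max, then added to emissive through Color_Time_of_Day.
# Base color blends toward Replace Color using the mask's red and blue channels (green is excluded), gated to night by Scalar_Time_of_Day.
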MaterialFunction /BRCosmetics/Accessories/FORT_Backpacks/Backpack_M_MED_BananaAdventure/Materials/MF_BananaAdventure_Backpack
graph {nodes [
  n1 = GetMaterialAttributes(None=FunctionInput() {inputName=Material, inputType=FunctionInput_MaterialAttributes}) {attributeGetTypes [#1, #2]}
  n2 = Constant3Vector()
  n3 = TextureSample() {
   samplerType=SAMPLERTYPE_Masks
   texture=/BRCosmetics/Accessories/FORT_Backpacks/Backpack_M_MED_BananaAdventure/Texture/T_BananaAdventure_Backpack_FX
  }
  n4 = FunctionInput(Preview=@n3.RGB) {inputName=Tech Art Mask, sortPriority=1}
  n5 = MaterialFunctionCall(Color=@n4) {materialFunction=&SplitComponents}
  n6 = Multiply(A=@n5.R, B=ScalarParameter() {defaultValue=1, group=Reactivity, parameterName=Red Mask Strength, sortPriority=0})
  n7 = VectorParameter() {group=Reactivity, parameterName=Red Mask Emissive, sortPriority=1, defaultValue=<rgba 0.2705 0.6724 0.402 1>}
  n8 = Multiply(A=@n5.G, B=ScalarParameter() {defaultValue=1, group=Reactivity, parameterName=Green Mask Strength, sortPriority=2})
  n9 = VectorParameter() {group=Reactivity, parameterName=Green Mask Emissive, sortPriority=3, defaultValue=<rgba 0.2705 0.6724 0.402 1>}
  n10 = MaterialFunctionCall(Color_Morning=@n2, Color_Day=@n2, Color_Evening=@n2, Color_Night=Max(A=Multiply(A=@n6, B=@n7.RGB), B=Multiply(A=@n8, B=@n9.RGB))) {materialFunction=&Color_Time_of_Day}
  n11 = NamedRerouteUsage()
  n12 = VectorParameter() {parameterName=Replace Color, defaultValue=<rgba 0.0745 0.1 0.0739 1>}
  n13 = MaterialFunctionCall(Color=@n4) {materialFunction=&SplitComponents}
  n14 = Constant()
  n15 = Multiply(
   A=Max(A=Max(A=@n13.R, B=@n13.R), B=@n13.B)
   B=MaterialFunctionCall(Value_Morning=@n14, Value_Day=@n14, Value_Evening=@n14, Value_Night=Constant() {r=1}) {materialFunction=&Scalar_Time_of_Day}.Result
  )
  FunctionOutput(None=SetMaterialAttributes!(
    MaterialAttributes=@n1.MaterialAttributes
    EmissiveColor=Add(A=@n1.EmissiveColor, B=@n10.Result) {constB=0}
    BaseColor=LinearInterpolate(A=@n11, B=@n12.RGB, Alpha=@n15)
   ) {attributeSetTypes [#1, #2]})
  NamedRerouteDeclaration(None=@n1.BaseColor) {name=BaseColor, nodeColor=<rgba 1 0 0.3516 1>}
 ]}
