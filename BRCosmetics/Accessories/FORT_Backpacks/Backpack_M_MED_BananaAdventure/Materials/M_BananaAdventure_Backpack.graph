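# Master material for the BananaAdventure backpack cosmetic, built on the shared 2023 character parent material functions.
# The PreFX parent function feeds a backpack-specific function, whose result enters the PostFX parent function alongside auxiliary PreFX data.
# It uses material attributes with an expression-driven shading model, including cloth, clear coat and subsurface, and supports skeletal meshes, Nanite and mesh deformers.
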Material /BRCosmetics/Accessories/FORT_Backpacks/Backpack_M_MED_BananaAdventure/Materials/M_BananaAdventure_Backpack
graph {
 properties {
  bUseMaterialAttributes=true
  bUsedWithMeshDeformer=true
  bUsedWithNanite=true
  bUsedWithSkeletalMesh=true
  pixelDepthOffsetMode=PDOM_Legacy
  shadingModel=MSM_FromMaterialExpression
  usedShadingModels=Default Lit | Subsurface | Clear Coat | Subsurface Profile | Cloth
 }
 nodes [
  n1 = MaterialFunctionCall() {materialFunction=&MF_CharacterParent_2023_PreFX}
  MaterialFunctionCall(
   Material=MaterialFunctionCall!(Material=@n1.Result, Tech Art Mask=@n1.Result) {materialFunction=&/MF_BananaAdventure_Backpack}.Result
   M=@n1.M
   Pre Skinned Local Position=@n1.M
   Distance Blend=@n1.M
   Fresnel=@n1.M
  ) {materialFunction=&MF_CharacterParent_2023_PostFX}
 ]
}
