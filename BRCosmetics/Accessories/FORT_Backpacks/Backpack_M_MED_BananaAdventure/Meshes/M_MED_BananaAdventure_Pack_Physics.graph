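# Physics asset for the Banana Adventure backpack cosmetic, defining collision and joints for its compass ring and arrow bones.
# Each of the two dynamic bones has a larger capsule body, constrained to a tiny capsule on its anchor bone (MSTR_compassRing or dyn_arrow_base).
# Both joints act as limited hinges with twist and one swing axis at zero, letting the ring and arrow swing on differing constraint axes.
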PhysicsAsset /BRCosmetics/Accessories/FORT_Backpacks/Backpack_M_MED_BananaAdventure/Meshes/M_MED_BananaAdventure_Pack_Physics
bodies {
 MSTR_compassRing [{shapeType=Capsule, length=0.505, radius=0.505, shapeName=MSTR_compassRing_capsule, rotation=<rot 0 -90 0>}]
 dyn_arrow [{
   shapeType=Capsule
   length=12.2209
   radius=2.5464
   shapeName=dyn_arrow_capsule
   center=<-0.0066 -0.0264 -0.0487>
   rotation=<rot 0 89.9909 -88.9782>
  }]
 dyn_arrow_base [{shapeType=Capsule, length=0.505, radius=0.505, shapeName=dyn_arrow_base_capsule, rotation=<rot 0 -90 0>}]
 dyn_compassRing [{shapeType=Capsule, length=8.3281, radius=2.5257, shapeName=dyn_compassRing_capsule, center=<10.022 -2.02 0>, rotation=<rot 0 -94.9859 0>}]
}
constraints [
 {bone1Name=dyn_arrow, bone2Name=dyn_arrow_base, swing1LimitDegrees=0, swing2LimitDegrees=45, twistLimitDegrees=0}
 {bone1Name=dyn_compassRing, bone2Name=MSTR_compassRing, swing1LimitDegrees=45, swing2LimitDegrees=0, twistLimitDegrees=0}
]
