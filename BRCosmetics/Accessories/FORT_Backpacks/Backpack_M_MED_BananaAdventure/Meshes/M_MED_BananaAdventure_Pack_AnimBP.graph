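# Anim Blueprint for the BananaAdventure backpack cosmetic, animating its compass arrow, dangling compass ring and pack body.
# The event graph converts root bone twist into an arrow rotation and maps spine bend to pack offset and roll, also setting movement flags the AnimGraph never reads.
# A RigidBody node simulates physics, with Anim Dynamics on ring and arrow when rigid bodies are disabled.
# A fixed ring rotation applies when Anim Dynamics is disabled.
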AnimBlueprint /BRCosmetics/Accessories/FORT_Backpacks/Backpack_M_MED_BananaAdventure/Meshes/M_MED_BananaAdventure_Pack_AnimBP
target_skeleton=&/M_MED_BananaAdventure_Pack_Skeleton
graphs [
 AnimGraph {nodes [
   n1 = "Transform (Modify) Bone - Bone: dyn_arrow_base"(
    ComponentPose=Local To Component(LocalPose=Input Pose())
    Alpha=1
    Translation="0,0,0"
    Rotation=Get cardinal() -> (cardinal:Rotator)
    Scale="1,1,1"
   ) {rotationMode=BMM_Replace, rotationSpace=BCS_BoneSpace, alphaBoolBlend {blendOption=Linear}, boneToModify {boneName=dyn_arrow_base}}
   n2 = RigidBody(ComponentPose=@n1) {
    bOverrideWorldGravity=true
    overridePhysicsAsset=&/M_MED_BananaAdventure_Pack_Physics
    simulationSpace=BaseBoneSpace
    alphaBoolBlend {blendOption=Linear}
    componentLinearAccScale {z=0.4}
    componentLinearVelScale {z=0.2}
    overrideWorldGravity {z=-980}
    simSpaceSettings {maxAngularAcceleration=1000, maxAngularVelocity=10, maxLinearAcceleration=500, maxLinearVelocity=500, velocityScaleZ=0, worldAlpha=1}
   }
   n3 = "Transform (Modify) Bone - Bone: pack"(ComponentPose=@n2, Translation=Get backpack_translation(), Rotation=Get backpack_rotate()) {
    rotationMode=BMM_Additive
    scaleMode=BMM_Additive
    scaleSpace=BCS_BoneSpace
    translationMode=BMM_Additive
    alphaBoolBlend {blendOption=Linear}
    boneToModify {boneName=pack}
   }
   n4 = NOT Boolean(A=Get bRigidBodyEnabled())
   n5 = Anim Dynamics - Bone: dyn_compassRing!(ComponentPose=@n3, bAlphaBoolEnabled:bool=@n4, ExternalForce="0,0,0") -> (Pose) {
    alphaInputType=Bool
    angularDampingOverride=0.9
    bOverrideAngularDamping=true
    bUseGravityOverride=true
    simulationSpace=BoneRelative
    alphaBoolBlend {blendOption=Linear}
    boundBone {boneName=dyn_compassRing}
    componentLinearAccScale {z=0.4}
    componentLinearVelScale {z=0.2}
    gravityOverride {z=-980}
    physicsBodyDefinitions [{
      boundBone {boneName=dyn_compassRing}
      boxExtents=<2 20 20>
      constraintSetup {angularLimitsMax=<0 0 -20>, angularLimitsMin=<0 0 -90>, angularTarget=<0 0 0>}
      localJointOffset=<5 0 0>
     }]
    relativeSpaceBone {boneName=root}
   }
   n6 = Anim Dynamics - Bone: dyn_arrow!(ComponentPose=@n5, bAlphaBoolEnabled:bool=@n4, ExternalForce="0,0,0") -> (Pose) {
    alphaInputType=Bool
    angularDampingOverride=0.9
    bOverrideAngularDamping=true
    bUseGravityOverride=true
    simulationSpace=BoneRelative
    alphaBoolBlend {blendOption=Linear}
    boundBone {boneName=dyn_arrow}
    componentLinearAccScale {y=0.5}
    componentLinearVelScale {y=0.25}
    gravityOverride {y=-980}
    physicsBodyDefinitions [{
      boundBone {boneName=dyn_arrow}
      boxExtents=<2 20 20>
      constraintSetup {angularLimitsMax=<0 90 0>, angularLimitsMin=<0 -80 0>, angularTarget=<0 0 0>}
      localJointOffset=<10 0 0>
     }]
    relativeSpaceBone {boneName=root}
   }
   n7 = "Transform (Modify) Bone - Bone: MSTR_compassRing"!(
    ComponentPose=@n6
    bAlphaBoolEnabled:bool=NOT Boolean(A=Get bAnimDynamicsEnabled() -> (bAnimDynamicsEnabled:bool))
    Translation="0,0,0"
    Rotation="0,-90,0"
    Scale="1,1,1"
   ) -> (Pose) {
    alphaInputType=Bool
    rotationMode=BMM_Additive
    rotationSpace=BCS_BoneSpace
    scaleMode=BMM_Additive
    scaleSpace=BCS_BoneSpace
    translationMode=BMM_Additive
    translationSpace=BCS_BoneSpace
    alphaBoolBlend {blendOption=Linear}
    boneToModify {boneName=MSTR_compassRing}
    rotation {yaw=-90}
   }
   Output Pose(Result=Component To Local(ComponentPose=@n7)) {layerGroup=None}
  ]}
 EventGraph {nodes [
   Event BlueprintUpdateAnimation()
   > n1 = PlayerMovement(AnimInstance=Self-Reference(), XDotProductCoefficient=1, YDotProductCoefficient=1, ZDotProductCoefficient=1)
   > n2 = Sequence!(execute) -> (then_0, then_1, then_2:Exec, then_3:Exec, then_4:Exec)
   > Set Move Back(Move Back:bool="float < float"(A=@n1.YMovement, B=-0.5)) -> (Output_Get:bool)
   Set Move Forward(execute=@n2.then_1, Move Forward:bool="float > float"(A=@n1.YMovement, B=0.5)) -> (Output_Get:bool)
   Set Move Right(execute=@n2.then_2, Move Right:bool="float > float"(A=@n1, B=0.5)) -> (Output_Get:bool)
   Set Move Left(execute=@n2.then_3, Move Left:bool="float < float"(A=@n1, B=-0.5)) -> (Output_Get:bool)
   n3 = Set direction(
    execute=@n2.then_4
    direction:double=GetAlphaFromReferenceTransformDelta(SkeletalMeshComponent=GetOwnerSkeletalMeshComponent(), BoneName=root, BaseName=None, InputRange=(X=0,Y=1), TwistAxis="1,0,0").RawOutput
   ) -> (Output_Get:double)
   n4 = float + float(A=@n3, B=180)
   > Set cardinal(cardinal:Rotator=MakeRotator(Pitch=float * float(A=@n4, B=-1) {is=title K2Node_PromotableOperator})) -> (Output_Get:Rotator)
   n5 = GetOwnerSkeletalMeshComponent()
   n6 = BreakRotator(InRot=BreakTransform(InTransform=GetDeltaTransformFromRefPose(self=@n5, BoneName=spine_05, BaseName=spine_01)).Rotation)
   > Set backpack_rotate(backpack_rotate=MakeRotator(Roll=MapRangeClamped(Value=@n6.Yaw, InRangeA=0, InRangeB=-30, OutRangeA=3, OutRangeB=25)))
   > Set backpack_translation(backpack_translation=MakeVector(Y=MapRangeClamped(Value=@n6.Yaw, InRangeA=0, InRangeB=-30, OutRangeB=-8)))
   TryGetPawnOwner()
   TryGetPawnOwner()
  ]}
]
variables [
 backpack_rotate {type=Rotator}
 backpack_translation {type=Vector}
 Cardinal {type=Rotator}
 Direction {type=number}
 Move Back {type=boolean}
 Move Forward {type=boolean}
 Move Left {type=boolean}
 Move Right {type=boolean}
]
